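# Material function for the CopperMoggy Swan Cyber skin that overlays an animated, pixelated grid-pattern effect on incoming material attributes.
# Two scrolling noises with floor-snapped UVs are combined and contrasted to index a gradient texture, then shaped by inverse Fresnel.
# Influence parameters control blends into base color, emissive, specular and roughness, plus normal flattening.
# A panning wave noise is also built, and a levels-adjusted term is added to TechArtMask's green channel as thin-film influence.
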MaterialFunction /BRCosmetics/Characters/Player/Female/Medium/Bodies/F_MED_CopperMoggy_Swan/Skins/Cyber/Materials/MF_CyberOverrides
graph {expressions {
  MaterialExpressionAdd_1 {
   class=MaterialExpressionAdd
   inputs [
    A {from=MaterialExpressionConvert_0}
    B {from=MaterialExpressionMultiply_12}
   ]
  }
  MaterialExpressionAdd_2 {
   class=MaterialExpressionAdd
   inputs [
    A {from=MaterialExpressionDivide_1}
    B {from=MaterialExpressionMultiply_5}
   ]
  }
  MaterialExpressionAdd_3 {
   class=MaterialExpressionAdd
   inputs [
    A {from=MaterialExpressionDivide_3}
    B {from=MaterialExpressionMultiply_10}
   ]
  }
  MaterialExpressionAdd_4 {
   class=MaterialExpressionAdd
   inputs [
    A {from=MaterialExpressionNamedRerouteUsage_13}
    B {from=MaterialExpressionMultiply_21}
   ]
  }
  MaterialExpressionAppendVector_0 {
   class=MaterialExpressionAppendVector
   inputs [
    A {from=MaterialExpressionVectorParameter_0, output=X}
    B {from=MaterialExpressionVectorParameter_0, output=X}
   ]
  }
  MaterialExpressionAppendVector_1 {
   class=MaterialExpressionAppendVector
   inputs [
    A {from=MaterialExpressionVectorParameter_1, output=X}
    B {from=MaterialExpressionVectorParameter_1, output=X}
   ]
  }
  MaterialExpressionAppendVector_2 {
   class=MaterialExpressionAppendVector
   inputs [
    A {from=MaterialExpressionVectorParameter_2, output=X}
    B {from=MaterialExpressionVectorParameter_2, output=X}
   ]
  }
  MaterialExpressionAppendVector_3 {
   class=MaterialExpressionAppendVector
   inputs [
    A {from=MaterialExpressionVectorParameter_3, output=X}
    B {from=MaterialExpressionVectorParameter_3, output=X}
   ]
  }
  MaterialExpressionConstant_0 {class=MaterialExpressionConstant, properties {r=0.908}}
  MaterialExpressionConstant_1 {class=MaterialExpressionConstant, properties {r=0.5}}
  MaterialExpressionConstant_2 {class=MaterialExpressionConstant, properties {r=0.738}}
  MaterialExpressionConstant_3 {class=MaterialExpressionConstant, properties {r=0.3}}
  MaterialExpressionConstant_4 {class=MaterialExpressionConstant, properties {r=0.665}}
  MaterialExpressionConstant_5 {class=MaterialExpressionConstant, properties {r=0.152}}
  MaterialExpressionConstant_6 {class=MaterialExpressionConstant, properties {r=0.5}}
  MaterialExpressionConstant_7 {class=MaterialExpressionConstant, properties {r=0.7}}
  MaterialExpressionConstant_8 {class=MaterialExpressionConstant}
  MaterialExpressionConstant_9 {class=MaterialExpressionConstant, properties {r=1}}
  MaterialExpressionConvert_0 {
   class=MaterialExpressionConvert
   inputs [None {from=MaterialExpressionFunctionInput_1}]
   properties {
    convertInputs [{
      type=Vector3
      expressionInput {
       expression=&self:MaterialExpressionFunctionInput_1
       inputName=None
       mask=0
       maskA=0
       maskB=0
       maskG=0
       maskR=0
       outputIndex=0
      }
     }]
    convertOutputs [{type=Scalar}, {type=Scalar}, {type=Scalar}]
   }
  }
  MaterialExpressionConvert_1 {
   class=MaterialExpressionConvert
   inputs [
    None {from=MaterialExpressionConvert_0}
    None {from=MaterialExpressionAdd_1}
    None {from=MaterialExpressionConvert_0}
   ]
   properties {
    convertInputs [
     {
      type=Scalar
      expressionInput {
       expression=&self:MaterialExpressionConvert_0
       inputName=None
       mask=0
       maskA=0
       maskB=0
       maskG=0
       maskR=0
       outputIndex=0
      }
     }
     {
      type=Scalar
      expressionInput {
       expression=&self:MaterialExpressionAdd_1
       inputName=None
       mask=0
       maskA=0
       maskB=0
       maskG=0
       maskR=0
       outputIndex=0
      }
     }
     {
      type=Scalar
      expressionInput {
       expression=&self:MaterialExpressionConvert_0
       inputName=None
       mask=0
       maskA=0
       maskB=0
       maskG=0
       maskR=0
       outputIndex=2
      }
     }
    ]
    convertOutputs [{type=Vector3}]
   }
  }
  MaterialExpressionDesaturation_0 {
   class=MaterialExpressionDesaturation
   inputs [
    None {from=MaterialExpressionMaterialFunctionCall_0, output=Result}
    Fraction {from="", output=""}
   ]
   properties {}
  }
  MaterialExpressionDesaturation_1 {
   class=MaterialExpressionDesaturation
   inputs [
    None {from=MaterialExpressionMaterialFunctionCall_2, output=Result}
    Fraction {from="", output=""}
   ]
   properties {}
  }
  MaterialExpressionDivide_0 {
   class=MaterialExpressionDivide
   inputs [
    A {from=MaterialExpressionFloor_0}
    B {from=MaterialExpressionNamedRerouteUsage_14}
   ]
  }
  MaterialExpressionDivide_1 {
   class=MaterialExpressionDivide
   inputs [
    A {from=MaterialExpressionFloor_1}
    B {from=MaterialExpressionNamedRerouteUsage_6}
   ]
  }
  MaterialExpressionDivide_3 {
   class=MaterialExpressionDivide
   inputs [
    A {from=MaterialExpressionFloor_3}
    B {from=MaterialExpressionNamedRerouteUsage_7}
   ]
  }
  MaterialExpressionFloor_0 {class=MaterialExpressionFloor, inputs [None {from=MaterialExpressionMultiply_3, output=""}], properties {}}
  MaterialExpressionFloor_1 {class=MaterialExpressionFloor, inputs [None {from=MaterialExpressionMultiply_4, output=""}], properties {}}
  MaterialExpressionFloor_3 {class=MaterialExpressionFloor, inputs [None {from=MaterialExpressionMultiply_9, output=""}], properties {}}
  MaterialExpressionFresnel_0 {
   class=MaterialExpressionFresnel
   inputs [
    ExponentIn {from="", output=""}
    BaseReflectFractionIn {from="", output=""}
    Normal {from="", output=""}
   ]
   properties {baseReflectFraction=0, exponent=2}
  }
  MaterialExpressionFunctionInput_0 {
   class=MaterialExpressionFunctionInput
   inputs [Preview {from="", output=""}]
   properties {inputName=Material, inputType=FunctionInput_MaterialAttributes}
  }
  MaterialExpressionFunctionInput_1 {
   class=MaterialExpressionFunctionInput
   inputs [Preview {from="", output=""}]
   properties {bUsePreviewValueAsDefault=true, inputName=TechArtMask, sortPriority=1}
  }
  MaterialExpressionFunctionOutput_0 {class=MaterialExpressionFunctionOutput, inputs [None {from=MaterialExpressionSetMaterialAttributes_0, output=""}]}
  MaterialExpressionFunctionOutput_1 {
   class=MaterialExpressionFunctionOutput
   inputs [None {from=MaterialExpressionConvert_1, output=""}]
   properties {outputName=TechArtMask, sortPriority=1}
  }
  MaterialExpressionGetMaterialAttributes_0 {
   class=MaterialExpressionGetMaterialAttributes
   inputs [None {from=MaterialExpressionFunctionInput_0, output=""}]
   properties {attributeGetTypes [
     #1
     #2
     #3
     #4
     #5
    ]}
  }
  MaterialExpressionLinearInterpolate_0 {
   class=MaterialExpressionLinearInterpolate
   inputs [
    A {from=""}
    B {from=MaterialExpressionPower_0}
    Alpha {from=MaterialExpressionScalarParameter_3}
   ]
   properties {constA=1}
  }
  MaterialExpressionLinearInterpolate_1 {
   class=MaterialExpressionLinearInterpolate
   inputs [
    A {from=""}
    B {from=MaterialExpressionTextureSample_2, output=R}
    Alpha {from=MaterialExpressionScalarParameter_4}
   ]
   properties {constA=1}
  }
  MaterialExpressionLinearInterpolate_2 {
   class=MaterialExpressionLinearInterpolate
   inputs [
    A {from=""}
    B {from=MaterialExpressionTextureSample_4, output=R}
    Alpha {from=MaterialExpressionScalarParameter_2}
   ]
   properties {constA=1}
  }
  MaterialExpressionLinearInterpolate_3 {
   class=MaterialExpressionLinearInterpolate
   inputs [
    A {from=MaterialExpressionGetMaterialAttributes_0, output=BaseColor}
    B {from=MaterialExpressionNamedRerouteUsage_3}
    Alpha {from=MaterialExpressionMultiply_19}
   ]
  }
  MaterialExpressionLinearInterpolate_4 {
   class=MaterialExpressionLinearInterpolate
   inputs [
    A {from=MaterialExpressionGetMaterialAttributes_0, output=Specular}
    B {from=MaterialExpressionDesaturation_0}
    Alpha {from=MaterialExpressionMultiply_0}
   ]
  }
  MaterialExpressionLinearInterpolate_5 {
   class=MaterialExpressionLinearInterpolate
   inputs [
    A {from=""}
    B {from=""}
    Alpha {from=MaterialExpressionNamedRerouteDeclaration_7}
   ]
  }
  MaterialExpressionLinearInterpolate_6 {
   class=MaterialExpressionLinearInterpolate
   inputs [
    A {from=MaterialExpressionGetMaterialAttributes_0, output=Roughness}
    B {from=MaterialExpressionDesaturation_1}
    Alpha {from=MaterialExpressionMultiply_2}
   ]
  }
  MaterialExpressionLinearInterpolate_8 {
   class=MaterialExpressionLinearInterpolate
   inputs [
    A {from=MaterialExpressionGetMaterialAttributes_0, output=EmissiveColor}
    B {from=MaterialExpressionMaterialFunctionCall_4, output=Result}
    Alpha {from=MaterialExpressionMultiply_20}
   ]
  }
  MaterialExpressionMaterialFunctionCall_0 {
   class=MaterialExpressionMaterialFunctionCall
   inputs [
    Color {from=MaterialExpressionNamedRerouteUsage_5}
    MinInput {from=MaterialExpressionConstant_1}
    Gamma {from=MaterialExpressionConstant_3}
    MaxInput {from=MaterialExpressionConstant_0}
   ]
   properties {materialFunction=/BRCosmetics/Characters/Player/Female/Medium/Bodies/F_MED_CopperMoggy_Swan/Skins/Prime/Materials/MF_Levels}
  }
  MaterialExpressionMaterialFunctionCall_1 {
   class=MaterialExpressionMaterialFunctionCall
   inputs [
    In {from=MaterialExpressionMultiply_15}
    Contrast {from=MaterialExpressionScalarParameter_7}
   ]
   properties {materialFunction=&CheapContrast}
  }
  MaterialExpressionMaterialFunctionCall_2 {
   class=MaterialExpressionMaterialFunctionCall
   inputs [
    Color {from=MaterialExpressionNamedRerouteUsage_12}
    MinInput {from=MaterialExpressionConstant_5}
    Gamma {from=MaterialExpressionConstant_6}
    MaxInput {from=MaterialExpressionConstant_4}
   ]
   properties {materialFunction=&MF_Levels}
  }
  MaterialExpressionMaterialFunctionCall_3 {
   class=MaterialExpressionMaterialFunctionCall
   inputs [
    Color {from=MaterialExpressionNamedRerouteUsage_15}
    MinInput {from=MaterialExpressionConstant_8}
    Gamma {from=MaterialExpressionConstant_9}
    MaxInput {from=MaterialExpressionConstant_7}
   ]
   properties {materialFunction=&MF_Levels}
  }
  MaterialExpressionMaterialFunctionCall_4 {
   class=MaterialExpressionMaterialFunctionCall
   inputs [
    Color {from=MaterialExpressionNamedRerouteUsage_4}
    MinInput {from=""}
    Gamma {from=MaterialExpressionConstant_2}
    MaxInput {from=""}
   ]
   properties {materialFunction=&MF_Levels}
  }
  MaterialExpressionMaterialFunctionCall_6 {
   class=MaterialExpressionMaterialFunctionCall
   inputs [
    Normal {from=MaterialExpressionGetMaterialAttributes_0, output=Normal}
    Flatness {from=MaterialExpressionMultiply_18}
   ]
   properties {materialFunction=&FlattenNormal}
  }
  MaterialExpressionMultiply_0 {
   class=MaterialExpressionMultiply
   inputs [
    A {from=MaterialExpressionScalarParameter_0}
    B {from=MaterialExpressionNamedRerouteUsage_2}
   ]
  }
  MaterialExpressionMultiply_1 {
   class=MaterialExpressionMultiply
   inputs [
    A {from=MaterialExpressionNamedRerouteDeclaration_2}
    B {from=MaterialExpressionNamedRerouteDeclaration_3}
   ]
  }
  MaterialExpressionMultiply_10 {
   class=MaterialExpressionMultiply
   inputs [
    A {from=MaterialExpressionTime_2}
    B {from=MaterialExpressionAppendVector_2}
   ]
  }
  MaterialExpressionMultiply_11 {
   class=MaterialExpressionMultiply
   inputs [
    A {from=MaterialExpressionSaturate_0}
    B {from=MaterialExpressionNamedRerouteDeclaration_4}
   ]
  }
  MaterialExpressionMultiply_12 {
   class=MaterialExpressionMultiply
   inputs [
    A {from=MaterialExpressionMaterialFunctionCall_3, output=Result}
    B {from=MaterialExpressionMultiply_7}
   ]
  }
  MaterialExpressionMultiply_14 {
   class=MaterialExpressionMultiply
   inputs [
    A {from=MaterialExpressionNamedRerouteUsage_0}
    B {from=MaterialExpressionScalarParameter_6}
   ]
   properties {constB=12}
  }
  MaterialExpressionMultiply_15 {
   class=MaterialExpressionMultiply
   inputs [
    A {from=MaterialExpressionLinearInterpolate_1}
    B {from=MaterialExpressionLinearInterpolate_2}
   ]
  }
  MaterialExpressionMultiply_16 {
   class=MaterialExpressionMultiply
   inputs [
    A {from=MaterialExpressionScalarParameter_14}
    B {from=MaterialExpressionNamedRerouteUsage_18}
   ]
  }
  MaterialExpressionMultiply_17 {
   class=MaterialExpressionMultiply
   inputs [
    A {from=MaterialExpressionNamedRerouteUsage_8}
    B {from=MaterialExpressionTextureSample_5, output=RGB}
   ]
  }
  MaterialExpressionMultiply_18 {
   class=MaterialExpressionMultiply
   inputs [
    A {from=MaterialExpressionNamedRerouteUsage_17}
    B {from=MaterialExpressionMultiply_16}
   ]
  }
  MaterialExpressionMultiply_19 {
   class=MaterialExpressionMultiply
   inputs [
    A {from=MaterialExpressionScalarParameter_11}
    B {from=MaterialExpressionNamedRerouteUsage_10}
   ]
  }
  MaterialExpressionMultiply_2 {
   class=MaterialExpressionMultiply
   inputs [
    A {from=MaterialExpressionScalarParameter_8}
    B {from=MaterialExpressionNamedRerouteUsage_9}
   ]
  }
  MaterialExpressionMultiply_20 {
   class=MaterialExpressionMultiply
   inputs [
    A {from=MaterialExpressionScalarParameter_12}
    B {from=MaterialExpressionNamedRerouteUsage_11}
   ]
  }
  MaterialExpressionMultiply_21 {
   class=MaterialExpressionMultiply
   inputs [
    A {from=MaterialExpressionNamedRerouteUsage_22}
    B {from=""}
   ]
   properties {constB=0.05}
  }
  MaterialExpressionMultiply_3 {
   class=MaterialExpressionMultiply
   inputs [
    A {from=MaterialExpressionMultiply_6}
    B {from=MaterialExpressionNamedRerouteUsage_14}
   ]
  }
  MaterialExpressionMultiply_4 {
   class=MaterialExpressionMultiply
   inputs [
    A {from=MaterialExpressionMultiply_8}
    B {from=MaterialExpressionNamedRerouteUsage_6}
   ]
  }
  MaterialExpressionMultiply_5 {
   class=MaterialExpressionMultiply
   inputs [
    A {from=MaterialExpressionTime_1}
    B {from=MaterialExpressionAppendVector_1}
   ]
  }
  MaterialExpressionMultiply_6 {
   class=MaterialExpressionMultiply
   inputs [
    A {from=MaterialExpressionAdd_4}
    B {from=MaterialExpressionAppendVector_3}
   ]
   properties {constB=12}
  }
  MaterialExpressionMultiply_7 {
   class=MaterialExpressionMultiply
   inputs [
    A {from=MaterialExpressionScalarParameter_10}
    B {from=MaterialExpressionNamedRerouteUsage_16}
   ]
  }
  MaterialExpressionMultiply_8 {
   class=MaterialExpressionMultiply
   inputs [
    A {from=MaterialExpressionNamedRerouteUsage_1}
    B {from=MaterialExpressionScalarParameter_5}
   ]
   properties {constB=12}
  }
  MaterialExpressionMultiply_9 {
   class=MaterialExpressionMultiply
   inputs [
    A {from=MaterialExpressionMultiply_14}
    B {from=MaterialExpressionNamedRerouteUsage_7}
   ]
  }
  MaterialExpressionNamedRerouteDeclaration_0 {
   class=MaterialExpressionNamedRerouteDeclaration
   inputs [None {from=MaterialExpressionTextureSample_1, output=R}]
   properties {name=PatternMask, nodeColor=<rgba 0.5703 0 1 1>}
  }
  MaterialExpressionNamedRerouteDeclaration_1 {
   class=MaterialExpressionNamedRerouteDeclaration
   inputs [None {from=MaterialExpressionMultiply_11}]
   properties {name=Noise_Output, nodeColor=<rgba 1 0 0.9141 1>}
  }
  MaterialExpressionNamedRerouteDeclaration_2 {
   class=MaterialExpressionNamedRerouteDeclaration
   inputs [None {from=MaterialExpressionTextureCoordinate_0}]
   properties {name=UVs, nodeColor=<rgba 1 0.9375 0 1>}
  }
  MaterialExpressionNamedRerouteDeclaration_3 {
   class=MaterialExpressionNamedRerouteDeclaration
   inputs [None {from=MaterialExpressionScalarParameter_1}]
   properties {name=PatternGridTile, nodeColor=<rgba 0.0312 0 1 1>}
  }
  MaterialExpressionNamedRerouteDeclaration_4 {
   class=MaterialExpressionNamedRerouteDeclaration
   inputs [None {from=MaterialExpressionMultiply_17}]
   properties {name=Color Result, nodeColor=<rgba 1 0 0.8672 1>}
  }
  MaterialExpressionNamedRerouteDeclaration_5 {
   class=MaterialExpressionNamedRerouteDeclaration
   inputs [None {from=MaterialExpressionSaturate_2}]
   properties {name=Wave Noise, nodeColor=<rgba 1 0 0.2812 1>}
  }
  MaterialExpressionNamedRerouteDeclaration_6 {
   class=MaterialExpressionNamedRerouteDeclaration
   inputs [None {from=MaterialExpressionLinearInterpolate_5}]
   properties {name=PatternRandomness_1, nodeColor=<rgba 0 0.7891 1 1>}
  }
  MaterialExpressionNamedRerouteDeclaration_7 {
   class=MaterialExpressionNamedRerouteDeclaration
   inputs [None {from=MaterialExpressionMaterialFunctionCall_1, output=Result}]
   properties {name=NoiseResult, nodeColor=<rgba 0 0.8359 1 1>}
  }
  MaterialExpressionNamedRerouteDeclaration_8 {
   class=MaterialExpressionNamedRerouteDeclaration
   inputs [None {from=MaterialExpressionTextureSample_1, output=B}]
   properties {name=PatternNoise, nodeColor=<rgba 1 0 0.4219 1>}
  }
  MaterialExpressionNamedRerouteUsage_0 {class=MaterialExpressionNamedRerouteUsage}
  MaterialExpressionNamedRerouteUsage_1 {class=MaterialExpressionNamedRerouteUsage}
  MaterialExpressionNamedRerouteUsage_10 {class=MaterialExpressionNamedRerouteUsage}
  MaterialExpressionNamedRerouteUsage_11 {class=MaterialExpressionNamedRerouteUsage}
  MaterialExpressionNamedRerouteUsage_12 {class=MaterialExpressionNamedRerouteUsage}
  MaterialExpressionNamedRerouteUsage_13 {class=MaterialExpressionNamedRerouteUsage}
  MaterialExpressionNamedRerouteUsage_14 {class=MaterialExpressionNamedRerouteUsage}
  MaterialExpressionNamedRerouteUsage_15 {class=MaterialExpressionNamedRerouteUsage}
  MaterialExpressionNamedRerouteUsage_16 {class=MaterialExpressionNamedRerouteUsage}
  MaterialExpressionNamedRerouteUsage_17 {class=MaterialExpressionNamedRerouteUsage}
  MaterialExpressionNamedRerouteUsage_18 {class=MaterialExpressionNamedRerouteUsage}
  MaterialExpressionNamedRerouteUsage_2 {class=MaterialExpressionNamedRerouteUsage}
  MaterialExpressionNamedRerouteUsage_22 {class=MaterialExpressionNamedRerouteUsage}
  MaterialExpressionNamedRerouteUsage_3 {class=MaterialExpressionNamedRerouteUsage}
  MaterialExpressionNamedRerouteUsage_4 {class=MaterialExpressionNamedRerouteUsage}
  MaterialExpressionNamedRerouteUsage_5 {class=MaterialExpressionNamedRerouteUsage}
  MaterialExpressionNamedRerouteUsage_6 {class=MaterialExpressionNamedRerouteUsage}
  MaterialExpressionNamedRerouteUsage_7 {class=MaterialExpressionNamedRerouteUsage}
  MaterialExpressionNamedRerouteUsage_8 {class=MaterialExpressionNamedRerouteUsage}
  MaterialExpressionNamedRerouteUsage_9 {class=MaterialExpressionNamedRerouteUsage}
  MaterialExpressionOneMinus_0 {class=MaterialExpressionOneMinus, inputs [None {from=MaterialExpressionFresnel_0}]}
  MaterialExpressionPanner_0 {
   class=MaterialExpressionPanner
   inputs [
    Coordinate {from=MaterialExpressionDivide_0}
    Time {from=""}
    Speed {from=MaterialExpressionAppendVector_0}
   ]
  }
  MaterialExpressionPower_0 {
   class=MaterialExpressionPower
   inputs [
    Base {from=MaterialExpressionOneMinus_0}
    Exp {from=""}
   ]
   properties {constExponent=3}
  }
  MaterialExpressionPower_1 {
   class=MaterialExpressionPower
   inputs [
    Base {from=MaterialExpressionMultiply_15}
    Exp {from=MaterialExpressionScalarParameter_7}
   ]
  }
  MaterialExpressionPower_3 {
   class=MaterialExpressionPower
   inputs [
    Base {from=MaterialExpressionTextureSample_11, output=R}
    Exp {from=MaterialExpressionScalarParameter_13}
   ]
  }
  MaterialExpressionSaturate_0 {class=MaterialExpressionSaturate, inputs [None {from=MaterialExpressionLinearInterpolate_0}]}
  MaterialExpressionSaturate_2 {class=MaterialExpressionSaturate, inputs [None {from=MaterialExpressionPower_3}]}
  MaterialExpressionScalarParameter_0 {class=MaterialExpressionScalarParameter, properties {defaultValue=0.18, group=CopperMoggy|Cyber, parameterName=Cyber_SpecularInfluence}}
  MaterialExpressionScalarParameter_1 {class=MaterialExpressionScalarParameter, properties {defaultValue=12, group=CopperMoggy|Cyber, parameterName=Cyber_PatternTiling}}
  MaterialExpressionScalarParameter_10 {class=MaterialExpressionScalarParameter, properties {defaultValue=0.18, group=CopperMoggy|Cyber, parameterName=Cyber_ThinFilmInfluence}}
  MaterialExpressionScalarParameter_11 {class=MaterialExpressionScalarParameter, properties {defaultValue=0.18, group=CopperMoggy|Cyber, parameterName=Cyber_BaseColorInfluence}}
  MaterialExpressionScalarParameter_12 {class=MaterialExpressionScalarParameter, properties {defaultValue=0.18, group=CopperMoggy|Cyber, parameterName=Cyber_EmissiveInfluence}}
  MaterialExpressionScalarParameter_13 {class=MaterialExpressionScalarParameter, properties {defaultValue=1, group=CopperMoggy|Cyber, parameterName=Cyber_WaiveNoise_Power}}
  MaterialExpressionScalarParameter_14 {class=MaterialExpressionScalarParameter, properties {defaultValue=0.18, group=CopperMoggy|Cyber, parameterName=Cyber_NormalInfluence}}
  MaterialExpressionScalarParameter_2 {class=MaterialExpressionScalarParameter, properties {defaultValue=0.6, group=CopperMoggy|Cyber, parameterName=Cyber_NoiseIntensity_B}}
  MaterialExpressionScalarParameter_3 {class=MaterialExpressionScalarParameter, properties {defaultValue=0.6, group=CopperMoggy|Cyber, parameterName=FresnelInfluence}}
  MaterialExpressionScalarParameter_4 {class=MaterialExpressionScalarParameter, properties {defaultValue=0.7, group=CopperMoggy|Cyber, parameterName=Cyber_NoiseIntensity_A}}
  MaterialExpressionScalarParameter_5 {class=MaterialExpressionScalarParameter, properties {defaultValue=8, group=CopperMoggy|Cyber, parameterName=Cyber_NoiseTiling_A}}
  MaterialExpressionScalarParameter_6 {class=MaterialExpressionScalarParameter, properties {defaultValue=16, group=CopperMoggy|Cyber, parameterName=Cyber_NoiseTiling_B}}
  MaterialExpressionScalarParameter_7 {class=MaterialExpressionScalarParameter, properties {defaultValue=0.05, group=CopperMoggy|Cyber, parameterName=Cyber_NoiseContrast}}
  MaterialExpressionScalarParameter_8 {class=MaterialExpressionScalarParameter, properties {defaultValue=0.18, group=CopperMoggy|Cyber, parameterName=Cyber_RougnessInfluence}}
  MaterialExpressionSetMaterialAttributes_0 {
   class=MaterialExpressionSetMaterialAttributes
   inputs [
    MaterialAttributes {from=MaterialExpressionGetMaterialAttributes_0, output=MaterialAttributes}
    BaseColor {from=MaterialExpressionLinearInterpolate_3, output=""}
    EmissiveColor {from=MaterialExpressionLinearInterpolate_8, output=""}
    Specular {from=MaterialExpressionLinearInterpolate_4, output=""}
    Roughness {from=MaterialExpressionLinearInterpolate_6, output=""}
    Normal {from=MaterialExpressionMaterialFunctionCall_6, output=Result}
   ]
   properties {attributeSetTypes [
     #1
     #2
     #3
     #4
     #5
    ]}
  }
  MaterialExpressionTextureCoordinate_0 {class=MaterialExpressionTextureCoordinate}
  MaterialExpressionTextureSample_0 {
   class=MaterialExpressionTextureSample
   inputs [
    UVs {from=""}
    Tex {from=""}
    Apply View MipBias {from=""}
   ]
   properties {
    samplerType=SAMPLERTYPE_Masks
    texture=/BRCosmetics/Characters/Player/Female/Medium/Bodies/F_MED_CopperMoggy_Swan/Skins/Cyber/Textures/T_CopperMoggy_Cyber_Body_S
   }
  }
  MaterialExpressionTextureSample_1 {
   class=MaterialExpressionTextureSample
   inputs [
    UVs {from=MaterialExpressionMultiply_1}
    Tex {from=""}
    Apply View MipBias {from=""}
   ]
   properties {
    samplerSource=SSM_Wrap_WorldGroupSettings
    samplerType=SAMPLERTYPE_LinearColor
    texture=/BRCosmetics/Characters/Player/Female/Medium/Bodies/F_MED_CopperMoggy_Swan/Skins/Prime/Textures/T_CopperMoggy_Prime_GridPattern
   }
  }
  MaterialExpressionTextureSample_11 {
   class=MaterialExpressionTextureSample
   inputs [
    UVs {from=MaterialExpressionPanner_0}
    Tex {from=""}
    Apply View MipBias {from=""}
   ]
   properties {samplerSource=SSM_Wrap_WorldGroupSettings, texture=/Game/Global/Textures/Default/Noise/T_SoftNoise}
  }
  MaterialExpressionTextureSample_2 {
   class=MaterialExpressionTextureSample
   inputs [
    UVs {from=MaterialExpressionAdd_2}
    Tex {from=""}
    Apply View MipBias {from=""}
   ]
   properties {samplerSource=SSM_Wrap_WorldGroupSettings, texture=/Game/Global/Textures/Default/Noise/TankMoteNoise}
  }
  MaterialExpressionTextureSample_4 {
   class=MaterialExpressionTextureSample
   inputs [
    UVs {from=MaterialExpressionAdd_3}
    Tex {from=""}
    Apply View MipBias {from=""}
   ]
   properties {samplerSource=SSM_Wrap_WorldGroupSettings, samplerType=SAMPLERTYPE_LinearColor, texture=/Game/Global/Textures/Default/Noise/T_CloudNoise_01}
  }
  MaterialExpressionTextureSample_5 {
   class=MaterialExpressionTextureSample
   inputs [
    UVs {from=MaterialExpressionNamedRerouteDeclaration_6}
    Tex {from=""}
    Apply View MipBias {from=""}
   ]
   properties {
    samplerSource=SSM_Wrap_WorldGroupSettings
    texture=/BRCosmetics/Characters/Player/Female/Medium/Bodies/F_MED_CopperMoggy_Swan/Skins/Cyber/Textures/T_CopperMoggy_Cyber_NoiseGradient
   }
  }
  MaterialExpressionTime_1 {class=MaterialExpressionTime}
  MaterialExpressionTime_2 {class=MaterialExpressionTime}
  MaterialExpressionVectorParameter_0 {
   class=MaterialExpressionVectorParameter
   properties {group=CopperMoggy|Cyber, parameterName=Cyber_WaveSpeed, channelNames {a=" ", b=" ", g=Y, r=X}, defaultValue=<rgba 0.1 -0.1 0 1>}
  }
  MaterialExpressionVectorParameter_1 {
   class=MaterialExpressionVectorParameter
   properties {group=CopperMoggy|Cyber, parameterName=Cyber_NoiseSpeed_A, channelNames {a=" ", b=" ", g=Y, r=X}, defaultValue=<rgba 0.0012 -0.0012 0 1>}
  }
  MaterialExpressionVectorParameter_2 {
   class=MaterialExpressionVectorParameter
   properties {group=CopperMoggy|Cyber, parameterName=Cyber_NoiseSpeed_B, channelNames {a=" ", b=" ", g=Y, r=X}, defaultValue=<rgba -0.0012 -0.0015 0 1>}
  }
  MaterialExpressionVectorParameter_3 {
   class=MaterialExpressionVectorParameter
   properties {group=CopperMoggy|Cyber, parameterName=Cyber_WaveTiling, channelNames {a=" ", b=" ", g=Y, r=X}, defaultValue=<rgba 8 8 0 1>}
  }
 }}
