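# Material function for the CopperMoggy Prime character skin that adds an animated digital mosaic glow to incoming material attributes' emissive.
# A grid texture, scrolled in quantized steps, supplies the tile mask and per-tile randomness that picks thin-film colors.
# Two pixelated, stepped noise layers and an inverted Fresnel modulate the glow.
# A Tech Art Mask input supports masking, and intermediate mask, color, grayscale and noise results are extra outputs.
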MaterialFunction /BRCosmetics/Characters/Player/Female/Medium/Bodies/F_MED_CopperMoggy_Swan/Skins/Prime/Materials/MF_CopperMoggy_DigitalMosaic
graph {nodes [
  n1 = FunctionInput() {inputName=Material, inputType=FunctionInput_MaterialAttributes}
  n2 = GetMaterialAttributes(None=@n1) {attributeGetTypes [#1]}
  n3 = VectorParameter() {group=DigitalMosaic, parameterName=PatternUVsOffset, channelNames {a=" ", b=" ", g=V Offset, r=U Offset}, defaultValue=<rgba 0 0 0 1>}
  n4 = NamedRerouteDeclaration(None=Add(A=TextureCoordinate(), B=AppendVector(A=@n3.U Offset, B=@n3.U Offset))) {name=UVs, nodeColor=<rgba 1 0.9375 0 1>}
  n5 = VectorParameter() {group=DigitalMosaic, parameterName=PatternSpeed, channelNames {a=" ", b=" ", g=Y, r=X}, defaultValue=<rgba 0.1 0.2 0 1>}
  n6 = Multiply(A=Time(), B=AppendVector(A=@n5.X, B=@n5.X))
  n7 = NamedRerouteDeclaration(None=ScalarParameter() {defaultValue=16, group=DigitalMosaic, parameterName=PatternSize}) {name=PatternSize, nodeColor=<rgba 0 0.7422 1 1>}
  n8 = Divide(A=Floor(None=Multiply(A=@n6, B=@n7)), B=@n7)
  n9 = TextureSample(UVs=Multiply(A=Add(A=@n4, B=@n8), B=ScalarParameter() {defaultValue=12, group=DigitalMosaic, parameterName=PatternTiling})) {
   samplerSource=SSM_Wrap_WorldGroupSettings
   samplerType=SAMPLERTYPE_LinearColor
   texture=/BRCosmetics/Characters/Player/Female/Medium/Bodies/F_MED_CopperMoggy_Swan/Skins/Prime/Textures/T_CopperMoggy_Prime_GridPattern
  }
  n10 = TextureSample(UVs=NamedRerouteDeclaration(None=LinearInterpolate(Alpha=@n9.B) {constB=0.6}) {name=PatternRandomness, nodeColor=<rgba 0 0.7891 1 1>}) {
   samplerSource=SSM_Wrap_WorldGroupSettings
   texture=/BRCosmetics/Characters/Player/Female/Medium/Bodies/F_MED_CopperMoggy_Swan/Skins/Prime/Textures/T_CopperMoggy_Prime_ThinFilm
  }
  n11 = Desaturation(
   None=Multiply(A=NamedRerouteDeclaration(None=@n9.R) {name=PatternMask, nodeColor=<rgba 0.5703 0 1 1>}, B=@n10.RGB)
   Fraction=ScalarParameter() {defaultValue=0.3, group=DigitalMosaic, parameterName=PatternDesaturation}
  )
  n12 = VectorParameter() {group=DigitalMosaic, parameterName=NoiseSpeed_A, channelNames {a=" ", b=" ", g=Y, r=X}, defaultValue=<rgba 0 -0.2 0 1>}
  n13 = NamedRerouteUsage()
  n14 = Multiply(A=Multiply(A=Time(), B=AppendVector(A=@n12.X, B=@n12.X)), B=@n13)
  n15 = Add(
   A=Multiply(A=NamedRerouteUsage(), B=ScalarParameter() {defaultValue=12, group=DigitalMosaic, parameterName=NoiseTiling_A}) {constB=12}
   B=Divide(A=Floor(None=@n14), B=@n13)
  )
  n16 = NamedRerouteUsage()
  n17 = Divide(A=Floor(None=Multiply(A=@n15, B=@n16)), B=@n16)
  n18 = NamedRerouteUsage()
  n19 = ScalarParameter() {defaultValue=6, group=DigitalMosaic, parameterName=NoiseTiling_B}
  n20 = Time()
  n21 = VectorParameter() {group=DigitalMosaic, parameterName=NoiseSpeed_B, channelNames {a=" ", b=" ", g=Y, r=X}, defaultValue=<rgba 0 -0.4 0 1>}
  n22 = AppendVector(A=@n21.X, B=@n21.X)
  n23 = NamedRerouteUsage()
  n24 = Floor(None=Multiply(A=Multiply(A=@n20, B=@n22), B=@n23))
  n25 = NamedRerouteUsage()
  n26 = Multiply(A=Add(A=Multiply(A=@n18, B=@n19) {constB=12}, B=Divide(A=@n24, B=@n23)), B=@n25)
  n27 = TextureSample(UVs=Divide(A=Floor(None=@n26), B=@n25)) {samplerSource=SSM_Wrap_WorldGroupSettings, samplerType=SAMPLERTYPE_LinearColor, texture=/Game/Global/Textures/Default/Noise/T_CloudNoise_01}
  n28 = Multiply(
   A=LinearInterpolate(
    B=TextureSample(UVs=@n17) {samplerSource=SSM_Wrap_WorldGroupSettings, texture=/Game/Global/Textures/Default/Noise/TankMoteNoise}.R
    Alpha=ScalarParameter() {defaultValue=0.7, group=DigitalMosaic, parameterName=NoiseIntensity_A}
   ) {constA=1}
   B=LinearInterpolate(B=@n27.R, Alpha=ScalarParameter() {defaultValue=0.8, group=DigitalMosaic, parameterName=NoiseIntensity_B}) {constA=1}
  )
  n29 = ScalarParameter() {defaultValue=1.5, group=DigitalMosaic, parameterName=NoiseContrast}
  n30 = Multiply(
   A=Multiply(
    A=NamedRerouteDeclaration(None=@n11) {name=Color Result, nodeColor=<rgba 1 0 0.8672 1>}
    B=ScalarParameter() {defaultValue=3, group=DigitalMosaic, parameterName=PatternIntensity}
   )
   B=NamedRerouteDeclaration(None=Power(Base=@n28, Exp=@n29)) {name=NoiseResult, nodeColor=<rgba 0 0.8359 1 1>}
  )
  n31 = OneMinus(None=Fresnel() {baseReflectFraction=0, exponent=2})
  n32 = Saturate(None=LinearInterpolate(B=Power(Base=@n31) {constExponent=3}, Alpha=ScalarParameter() {defaultValue=0.6, group=DigitalMosaic, parameterName=FresnelInfluence}) {constA=1})
  n33 = Add(
   A=@n2.EmissiveColor
   B=Multiply(A=Multiply(A=@n30, B=@n32), B=ScalarParameter() {defaultValue=1, group=DigitalMosaic, parameterName=EmissiveIntensity})
  )
  FunctionOutput(None=BlendMaterialAttributes(
    A=@n1
    B=SetMaterialAttributes!(MaterialAttributes=@n2.MaterialAttributes, EmissiveColor=@n33) {attributeSetTypes [#1]}
    Alpha=NamedRerouteUsage()
   )) {outputName=Material Result}
  FunctionOutput(None=NamedRerouteUsage()) {outputName=GridMask, sortPriority=1}
  n34 = NamedRerouteUsage()
  FunctionOutput(None=@n34) {outputName=Color Result, sortPriority=2}
  FunctionOutput(None=Desaturation(None=@n34)) {outputName=Grayscale Result, sortPriority=3}
  FunctionOutput(None=NamedRerouteUsage()) {outputName=PatternRandomness, sortPriority=5}
  FunctionOutput(None=NamedRerouteUsage()) {outputName=NoiseResult, sortPriority=4}
  n35 = Constant3Vector() {constant=<rgba 1 1 1 0>}
  NamedRerouteDeclaration(None=StaticComponentMaskParameter(None=FunctionInput(Preview=@n35) {bUsePreviewValueAsDefault=true, inputName=Tech Art Mask, sortPriority=1}) {defaultR=true, group=DigitalMosaic, parameterName=MosaicMaskComponent}) {name=EffectMask, nodeColor=<rgba 0.5703 0 1 1>}
  n36 = VectorParameter() {group=DigitalMosaic, parameterName=NoiseUVsOffset, channelNames {a=" ", b=" ", g=V Offset, r=U Offset}, defaultValue=<rgba 0 0 0 1>}
  NamedRerouteDeclaration(None=Add(A=NamedRerouteUsage(), B=AppendVector(A=@n36.U Offset, B=@n36.U Offset))) {name=UVOffset, nodeColor=<rgba 0 1 0.2734 1>}
 ]}
