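# Translucent parent glass material for the CopperMoggy Swan character cosmetic, with static switches for Cyber, DigitalMosaic and Flame style variants.
# MF_Glass feeds optional Cyber overrides, thin film, optional camera dither, then either vertex-collapse hiding via world position offset or standard character opacity.
# Optional mosaic and flame layers follow, before translucency-sorted opacity and final TechArt overrides.
# The flame layer's temporal mask drives Temporal Responsiveness when enabled.
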Material /BRCosmetics/Characters/Player/Female/Medium/Bodies/F_MED_CopperMoggy_Swan/Skins/Prime/Materials/M_CopperMoggy_Glass_Parent
graph {
 properties {
  allowTranslucentCustomDepthWrites=true
  bEnableResponsiveAA=true
  bOutputTranslucentVelocity=true
  bScreenSpaceReflections=true
  bUseMaterialAttributes=true
  bUsedWithMeshDeformer=true
  bUsedWithMorphTargets=true
  bUsedWithSkeletalMesh=true
  blendMode=BLEND_Translucent
  ditherOpacityMask=true
  pixelDepthOffsetMode=PDOM_Legacy
  translucencyDirectionalLightingIntensity=5
  translucencyLightingMode=TLM_SurfacePerPixelLighting
  usedShadingModels=Default Lit
 }
 nodes [
  n1 = TextureSampleParameter2D() {
   group=00_Glass
   parameterName=TechArtMask
   samplerSource=SSM_Clamp_WorldGroupSettings
   samplerType=SAMPLERTYPE_Masks
   sortPriority=-96
   texture=&T_BlackMask
  }
  n2 = MaterialFunctionCall!(TechArtMask=NamedRerouteDeclaration(None=@n1.RGB) {name=TechArtMask, nodeColor=<rgba 0 1 0.3203 1>}) {materialFunction=/Game/Global/Materials/Character_Parent_2023/Glass_Parent_2023/MF_Glass}
  n3 = MaterialFunctionCall!(Material=@n2.Result, TechArtMask=NamedRerouteUsage()) {materialFunction=/BRCosmetics/Characters/Player/Female/Medium/Bodies/F_MED_CopperMoggy_Swan/Skins/Cyber/Materials/MF_CyberOverrides}
  n4 = MaterialFunctionCall!(
   Material Attributes=StaticSwitch(True=@n3.Result, False=@n2.Result, Value=StaticBoolParameter() {parameterName=EnableCyberOverrides})
   Fresnel=Fresnel() {baseReflectFraction=0, exponent=1}
   Alpha=StaticSwitch(True=@n3.TechArtMask, False=NamedRerouteUsage(), Value=StaticBoolParameter() {parameterName=EnableCyberOverrides})
  ) {materialFunction=/Game/Global/MaterialFunctions/TechArt/MF_ThinFilm_TechArt}
  n5 = StaticSwitchParameter(
   True=MaterialFunctionCall!(In=@n4.Result) {materialFunction=/Game/Packages/Fortress_SharedMaterials/Base_Material_Functions/BaseCustomizationFunctions/MF_FN_Vehicle_CameraDither}.Result
   False=@n4.Result
  ) {group=00_Glass, parameterName=UseCameraDither, sortPriority=250}
  n6 = Subtract(A=ObjectPositionWS(), B=WorldPosition().XYZ)
  n7 = ScalarParameter() {group=00_Glass, parameterName=VertexHide, sortPriority=301}
  n8 = StaticSwitchParameter(
   True=SetMaterialAttributes!(MaterialAttributes=@n5, World Position Offset=Multiply(A=@n6, B=@n7)) {attributeSetTypes [#1]}
   False=MaterialFunctionCall!(Material=@n5) {materialFunction=/Game/Global/Materials/Character_Parent_2023/Functions/BaseFN_CharacterOpacity_2023}.Result
  ) {group=00_Glass, parameterName=UseVertexHide, sortPriority=300}
  n9 = TextureSampleParameter2D() {
   group=DigitalMosaic
   parameterName=TechArtMask_Additional
   samplerSource=SSM_Clamp_WorldGroupSettings
   samplerType=SAMPLERTYPE_Masks
   sortPriority=0
   texture=&T_WhiteMask
  }
  n10 = MaterialFunctionCall!(Material=@n8, Tech Art Mask=@n9.RGB) {materialFunction=&/MF_CopperMoggy_DigitalMosaic}
  n11 = StaticSwitchParameter(True=@n10.Material Result, False=@n8) {group=DigitalMosaic, parameterName=EnableDigitalMosaic}
  n12 = MaterialFunctionCall!(Material=@n11, TechArtMask=NamedRerouteUsage()) {materialFunction=/BRCosmetics/Characters/Player/Female/Medium/Bodies/F_MED_CopperMoggy_Swan/Skins/Flame/Materials/MF_CopperMoggy_FlameOverrides}
  n13 = GetMaterialAttributes(None=StaticSwitchParameter(True=@n12.Result, False=@n11) {group=Flame, parameterName=EnableFlameOverrides}) {attributeGetTypes [#2]}
  MaterialFunctionCall(In=SetMaterialAttributes!(
    MaterialAttributes=@n13.MaterialAttributes
    Opacity=MaterialFunctionCall!(Opacity=@n13.Opacity) {materialFunction=/Game/Global/Materials/MaterialGym/Materials/SortedOutline/MF_SortTranslucency}.Result
   ) {attributeSetTypes [#2]}) {materialFunction=&MF_TechArtOverrides}
  TemporalResponsivenessOutput!(None=StaticSwitchParameter(True=@n12.Temporal Mask, False=Constant()) {group=Flame, parameterName=EnableFlameOverrides})
 ]
}
